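# Control Rig for the PollenTrove pickaxe that procedurally animates three bees with intermittent wing buzzing and drifting bodies.
# Per-bee countdown timers toggle shaking with random durations and strengths, scaling sine waves that drive spring-smoothed wing rotations.
# Random targets feed springs that jitter each body's translation.
# All three timer passes add delta time to the shared Time variable, so the sine clock advances at triple speed.
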ControlRigBlueprint /BRCosmetics/Weapons/FORT_Melee/Pickaxe_PollenTrove/Meshes/PollenTrove_Axe_CtrlRig
graphs [
 RigVMFunctionLibrary {links [], nodes []}
 RigVMFunction_Sequence_1_ContainedGraph {nodes [
   n1 = RigVMFunctionEntryNode()
   RigVMFunction_Sequence = &RigVMFunction_Sequence(ExecuteContext=@n1.ExecuteContext)
   n2 = &RigVMFunction_Sequence(ExecuteContext=@RigVMFunction_Sequence.B)
   n3 = &RigVMFunction_Sequence(ExecuteContext=@n2.B)
   n4 = &RigVMFunction_Sequence(ExecuteContext=@n3.B)
   RigVMFunctionReturnNode(A=@RigVMFunction_Sequence.A, B=@n2.A, C=@n3.A, D=@n4.A, E=@n4.B)
  ]}
 RigVMModel {nodes [
   RigUnit_BeginExecution = &RigUnit_BeginExecution()
   RigVMFunction_Sequence = RigVMAggregateNode(ExecuteContext=@RigUnit_BeginExecution.ExecutePin)
   Multiply = &RigVMFunction_MathDoubleMul(B=40, A=RigVMVariableNode() {variable=Time}.Value)
   Sin = /Script/RigVM.RigVMFunction_MathDoubleSin(Value=@Multiply.Result) {is=struct RigVMUnitNode, template="Sin::Execute(in Value,out Result)"}
   n1 = RigVMVariableNode() {variable=ShakeAlpha1}
   RigVMVariableNode(
    Value=&RigVMFunction_MathDoubleMul(B=&RigVMFunction_MathDoubleMul(B=@Sin.Result, A=@n1.Value).Result, A=RigVMVariableNode() {variable=MaxAngle1}.Value).Result
    ExecuteContext=@RigVMFunction_Sequence.D
   ) {variable=ShakeValue1}
   n2 = &RigVMFunction_MathDoubleSin(Value=&RigVMFunction_MathDoubleMul(B=30, A=RigVMVariableNode() {variable=Time}.Value).Result) {template="Sin::Execute(in Value,out Result)"}
   n3 = RigVMVariableNode() {variable=ShakeAlpha2}
   > RigVMVariableNode(Value=&RigVMFunction_MathDoubleMul(B=&RigVMFunction_MathDoubleMul(B=@n2.Result, A=@n3.Value).Result, A=RigVMVariableNode() {variable=MaxAngle2}.Value).Result) {variable=ShakeValue2}
   n4 = &RigVMFunction_MathDoubleSin(Value=&RigVMFunction_MathDoubleMul(B=30, A=RigVMVariableNode() {variable=Time}.Value).Result) {template="Sin::Execute(in Value,out Result)"}
   n5 = RigVMVariableNode() {variable=ShakeAlpha3}
   > RigVMVariableNode(Value=&RigVMFunction_MathDoubleMul(B=&RigVMFunction_MathDoubleMul(B=@n4.Result, A=@n5.Value).Result, A=RigVMVariableNode() {variable=MaxAngle3}.Value).Result) {variable=ShakeValue3}
   Add = &RigVMFunction_MathDoubleAdd(B=/Script/RigVM.RigVMFunction_GetDeltaTime() {is=struct RigVMUnitNode}.Result, A=RigVMVariableNode() {variable=Time}.Value)
   RigVMVariableNode(Value=@Add.Result, ExecuteContext=@RigVMFunction_Sequence.A) {variable=Time}
   Subtract = &RigVMFunction_MathDoubleSub(B=&RigVMFunction_GetDeltaTime().Result, A=RigVMVariableNode() {variable=ShakeTimer1}.Value)
   > n6 = RigVMVariableNode(Value=@Subtract.Result) {variable=ShakeTimer1}
   LessEqual = /Script/RigVM.RigVMFunction_MathDoubleLessEqual(B=0, A=RigVMVariableNode() {variable=ShakeTimer1}.Value) {is=struct RigVMUnitNode, template="LessEqual::Execute(in A,in B,out Result)"}
   n7 = &RigVMFunction_ControlFlowBranch(Condition=@LessEqual.Result, ExecuteContext=@n6.ExecuteContext)
   RigVMFunction_MathBoolNot = &RigVMFunction_MathBoolNot(Value=RigVMVariableNode() {variable=IsShaking1}.Value)
   n8 = RigVMVariableNode(ExecuteContext=@n7.True, Value=@RigVMFunction_MathBoolNot.Result) {variable=IsShaking1}
   n9 = &RigVMFunction_ControlFlowBranch(ExecuteContext=@n8.ExecuteContext, Condition=RigVMVariableNode() {variable=IsShaking1}.Value)
   RigVMVariableNode(ExecuteContext=@n9.False) {variable=ShakeAlpha1}
   > RigVMVariableNode(Value=/Script/RigVM.RigVMFunction_RandomFloat(Duration=0, Maximum=1.8, Minimum=0.3, Seed=217) {is=struct RigVMUnitNode, template="Random::Execute(in Seed,in Minimum,in Maximum,in Duration,out Result)"}.Result) {variable=ShakeTimer1}
   Random = &RigVMFunction_RandomFloat(Duration=0, Maximum=1.7, Minimum=0.4, Seed=217) {template="Random::Execute(in Seed,in Minimum,in Maximum,in Duration,out Result)"}
   RigVMVariableNode(Value=@Random.Result, ExecuteContext=@n9.True) {variable=ShakeAlpha1}
   > RigVMVariableNode(Value=&RigVMFunction_RandomFloat(Duration=0, Maximum=1.2, Minimum=0.08, Seed=217) {template="Random::Execute(in Seed,in Minimum,in Maximum,in Duration,out Result)"}.Result) {variable=ShakeTimer1}
   n10 = /Script/ControlRig.RigUnit_SpringInterpVectorV2(
    CriticalDamping=0.5
    Current=(X=0,Y=0,Z=0)
    Force=(X=0,Y=1,Z=1)
    Strength=10
    Target=(X=0,Y=0,Z=0)
    TargetVelocityAmount=0
    bInitializeFromTarget=False
    bUseCurrentInput=False
    Target.Z=RigVMVariableNode() {variable=ShakeValue1}.Value
   ) {
    is=struct RigVMUnitNode
    template="SpringInterp::Execute(in Target,in Strength,in CriticalDamping,in Force,in bUseCurrentInput,in Current,in TargetVelocityAmount,in bInitializeFromTarget,out Result,out Velocity)"
   }
   &RigUnit_ModifyTransforms(
    ItemToModify=((Item=(Type=Bone,Name="dyn_bee_01_wing_r"),Transform=(Rotation=(X=0,Y=0,Z=0,W=1),Translation=(X=0,Y=0,Z=0),Scale3D=(X=1,Y=1,Z=1))))
    Weight=1
    WeightMinimum=0
    ItemToModify.0.Transform.Rotation=&RigVMFunction_MathQuaternionFromEuler(Euler=@n10.Result).Result
    ExecutePin=@RigVMFunction_Sequence.E
   )
   n11 = &RigUnit_SpringInterpVectorV2(
    CriticalDamping=0.5
    Current=(X=0,Y=0,Z=0)
    Force=(X=0,Y=1,Z=1)
    Strength=10
    Target=(X=0,Y=0,Z=0)
    TargetVelocityAmount=0
    bInitializeFromTarget=False
    bUseCurrentInput=False
    Target.Z=RigVMVariableNode() {variable=ShakeValue2}.Value
   ) {template="SpringInterp::Execute(in Target,in Strength,in CriticalDamping,in Force,in bUseCurrentInput,in Current,in TargetVelocityAmount,in bInitializeFromTarget,out Result,out Velocity)"}
   > &RigUnit_ModifyTransforms(
    ItemToModify=((Item=(Type=Bone,Name="dyn_bee_02_wing_l"),Transform=(Rotation=(X=0,Y=0,Z=0,W=1),Translation=(X=0,Y=0,Z=0),Scale3D=(X=1,Y=1,Z=1))))
    Weight=1
    WeightMinimum=0
    ItemToModify.0.Transform.Rotation=&RigVMFunction_MathQuaternionFromEuler(Euler=@n11.Result).Result
   )
   n12 = &RigUnit_SpringInterpVectorV2(
    CriticalDamping=0.4
    Current=(X=0,Y=0,Z=0)
    Force=(X=0,Y=1,Z=1)
    Strength=8
    Target=(X=0,Y=0,Z=0)
    TargetVelocityAmount=0
    bInitializeFromTarget=False
    bUseCurrentInput=False
    Target.Z=RigVMVariableNode() {variable=ShakeValue3}.Value
   ) {template="SpringInterp::Execute(in Target,in Strength,in CriticalDamping,in Force,in bUseCurrentInput,in Current,in TargetVelocityAmount,in bInitializeFromTarget,out Result,out Velocity)"}
   > &RigUnit_ModifyTransforms(
    ItemToModify=((Item=(Type=Bone,Name="dyn_bee_03_wing_r"),Transform=(Rotation=(X=0,Y=0,Z=0,W=1),Translation=(X=0,Y=0,Z=0),Scale3D=(X=1,Y=1,Z=1))))
    Weight=1
    WeightMinimum=0
    ItemToModify.0.Transform.Rotation=&RigVMFunction_MathQuaternionFromEuler(Euler=@n12.Result).Result
   )
   > &RigUnit_ModifyTransforms(
    ItemToModify="((Item=(Type=Bone,Name=\"dyn_bee_01_wing_r\"),Transform=(Rotation=(X=0,Y=0,Z=0.0872,W=0.9962),Translation=(X=0,Y=0,Z=0),Scale3D=(X=1,Y=1,Z=1))),(Item=(Type=Bone,Name=\"dyn_bee_02_wing_l\"),Transform=(Rotation=(X=0,Y=0,Z=0.0872,W=0.9962),Translation=(X=0,Y=0,Z= ...(truncated)"
    Weight=1
    WeightMinimum=0
   )
   n13 = &RigUnit_SpringInterpVectorV2(
    CriticalDamping=0.5
    Current=(X=0,Y=0,Z=0)
    Force=(X=0,Y=1,Z=1)
    Strength=3
    Target=(X=0,Y=0,Z=0)
    TargetVelocityAmount=0
    bInitializeFromTarget=False
    bUseCurrentInput=False
    Target.Z=&RigVMFunction_RandomFloat(Duration=1, Maximum=0.1, Minimum=-0.1, Seed=641) {template="Random::Execute(in Seed,in Minimum,in Maximum,in Duration,out Result)"}.Result
    Target.Y=&RigVMFunction_RandomFloat(Duration=1, Maximum=0.1, Minimum=-0.1, Seed=217) {template="Random::Execute(in Seed,in Minimum,in Maximum,in Duration,out Result)"}.Result
   ) {template="SpringInterp::Execute(in Target,in Strength,in CriticalDamping,in Force,in bUseCurrentInput,in Current,in TargetVelocityAmount,in bInitializeFromTarget,out Result,out Velocity)"}
   > &RigUnit_ModifyTransforms(
    ItemToModify=((Item=(Type=Bone,Name="dyn_bee_01_body"),Transform=(Rotation=(X=0,Y=0,Z=0,W=1),Translation=(X=0,Y=0,Z=0),Scale3D=(X=1,Y=1,Z=1))))
    Weight=1
    WeightMinimum=0
    ItemToModify.0.Transform.Translation=@n13.Result
   )
   n14 = &RigUnit_SpringInterpVectorV2(
    CriticalDamping=0.5
    Current=(X=0,Y=0,Z=0)
    Force=(X=0,Y=1,Z=1)
    Strength=3
    Target=(X=0,Y=0,Z=0)
    TargetVelocityAmount=0
    bInitializeFromTarget=False
    bUseCurrentInput=False
    Target.Z=&RigVMFunction_RandomFloat(Duration=1.3, Maximum=0.1, Minimum=-0.1, Seed=230) {template="Random::Execute(in Seed,in Minimum,in Maximum,in Duration,out Result)"}.Result
    Target.Y=&RigVMFunction_RandomFloat(Duration=1.3, Maximum=0.1, Minimum=-0.1, Seed=650) {template="Random::Execute(in Seed,in Minimum,in Maximum,in Duration,out Result)"}.Result
   ) {template="SpringInterp::Execute(in Target,in Strength,in CriticalDamping,in Force,in bUseCurrentInput,in Current,in TargetVelocityAmount,in bInitializeFromTarget,out Result,out Velocity)"}
   > &RigUnit_ModifyTransforms(
    ItemToModify=((Item=(Type=Bone,Name="dyn_bee_02_body"),Transform=(Rotation=(X=0,Y=0,Z=0,W=1),Translation=(X=0,Y=0,Z=0),Scale3D=(X=1,Y=1,Z=1))))
    Weight=1
    WeightMinimum=0
    ItemToModify.0.Transform.Translation=@n14.Result
   )
   n15 = &RigUnit_SpringInterpVectorV2(
    CriticalDamping=0.5
    Current=(X=0,Y=0,Z=0)
    Force=(X=0,Y=1,Z=1)
    Strength=3
    Target=(X=0,Y=0,Z=0)
    TargetVelocityAmount=0
    bInitializeFromTarget=False
    bUseCurrentInput=False
    Target.Y=&RigVMFunction_RandomFloat(Duration=0.9, Maximum=0.1, Minimum=-0.1, Seed=230) {template="Random::Execute(in Seed,in Minimum,in Maximum,in Duration,out Result)"}.Result
    Target.Z=&RigVMFunction_RandomFloat(Duration=0.9, Maximum=0.1, Minimum=-0.1, Seed=650) {template="Random::Execute(in Seed,in Minimum,in Maximum,in Duration,out Result)"}.Result
   ) {template="SpringInterp::Execute(in Target,in Strength,in CriticalDamping,in Force,in bUseCurrentInput,in Current,in TargetVelocityAmount,in bInitializeFromTarget,out Result,out Velocity)"}
   > &RigUnit_ModifyTransforms(
    ItemToModify=((Item=(Type=Bone,Name="dyn_bee_03_body"),Transform=(Rotation=(X=0,Y=0,Z=0,W=1),Translation=(X=0,Y=0,Z=0),Scale3D=(X=1,Y=1,Z=1))))
    Weight=1
    WeightMinimum=0
    ItemToModify.0.Transform.Translation=@n15.Result
   )
   RigVMVariableNode(
    Value=&RigVMFunction_MathDoubleAdd(B=&RigVMFunction_GetDeltaTime().Result, A=RigVMVariableNode() {variable=Time}.Value).Result
    ExecuteContext=@RigVMFunction_Sequence.B
   ) {variable=Time}
   > n16 = RigVMVariableNode(Value=&RigVMFunction_MathDoubleSub(B=&RigVMFunction_GetDeltaTime().Result, A=RigVMVariableNode() {variable=ShakeTimer2}.Value).Result) {variable=ShakeTimer2}
   n17 = &RigVMFunction_ControlFlowBranch(
    Condition=&RigVMFunction_MathDoubleLessEqual(B=0, A=RigVMVariableNode() {variable=ShakeTimer2}.Value) {template="LessEqual::Execute(in A,in B,out Result)"}.Result
    ExecuteContext=@n16.ExecuteContext
   )
   n18 = RigVMVariableNode(ExecuteContext=@n17.True, Value=&RigVMFunction_MathBoolNot(Value=RigVMVariableNode() {variable=IsShaking2}.Value).Result) {variable=IsShaking2}
   n19 = &RigVMFunction_ControlFlowBranch(ExecuteContext=@n18.ExecuteContext, Condition=RigVMVariableNode() {variable=IsShaking2}.Value)
   RigVMVariableNode(ExecuteContext=@n19.False) {variable=ShakeAlpha2}
   > RigVMVariableNode(Value=&RigVMFunction_RandomFloat(Duration=0, Maximum=2.2, Minimum=0.5, Seed=217) {template="Random::Execute(in Seed,in Minimum,in Maximum,in Duration,out Result)"}.Result) {variable=ShakeTimer2}
   RigVMVariableNode(
    Value=&RigVMFunction_RandomFloat(Duration=0, Maximum=1.5, Minimum=0.2, Seed=217) {template="Random::Execute(in Seed,in Minimum,in Maximum,in Duration,out Result)"}.Result
    ExecuteContext=@n19.True
   ) {variable=ShakeAlpha2}
   > RigVMVariableNode(Value=&RigVMFunction_RandomFloat(Duration=0, Maximum=1.35, Minimum=0.15, Seed=217) {template="Random::Execute(in Seed,in Minimum,in Maximum,in Duration,out Result)"}.Result) {variable=ShakeTimer2}
   RigVMVariableNode(
    Value=&RigVMFunction_MathDoubleAdd(B=&RigVMFunction_GetDeltaTime().Result, A=RigVMVariableNode() {variable=Time}.Value).Result
    ExecuteContext=@RigVMFunction_Sequence.C
   ) {variable=Time}
   > n20 = RigVMVariableNode(Value=&RigVMFunction_MathDoubleSub(B=&RigVMFunction_GetDeltaTime().Result, A=RigVMVariableNode() {variable=ShakeTimer3}.Value).Result) {variable=ShakeTimer3}
   n21 = &RigVMFunction_ControlFlowBranch(
    Condition=&RigVMFunction_MathDoubleLessEqual(B=0, A=RigVMVariableNode() {variable=ShakeTimer3}.Value) {template="LessEqual::Execute(in A,in B,out Result)"}.Result
    ExecuteContext=@n20.ExecuteContext
   )
   n22 = RigVMVariableNode(ExecuteContext=@n21.True, Value=&RigVMFunction_MathBoolNot(Value=RigVMVariableNode() {variable=IsShaking3}.Value).Result) {variable=IsShaking3}
   n23 = &RigVMFunction_ControlFlowBranch(ExecuteContext=@n22.ExecuteContext, Condition=RigVMVariableNode() {variable=IsShaking3}.Value)
   RigVMVariableNode(ExecuteContext=@n23.False) {variable=ShakeAlpha3}
   > RigVMVariableNode(Value=&RigVMFunction_RandomFloat(Duration=0, Maximum=2, Minimum=0.3, Seed=217) {template="Random::Execute(in Seed,in Minimum,in Maximum,in Duration,out Result)"}.Result) {variable=ShakeTimer3}
   RigVMVariableNode(
    Value=&RigVMFunction_RandomFloat(Duration=0, Maximum=2, Minimum=0.5, Seed=217) {template="Random::Execute(in Seed,in Minimum,in Maximum,in Duration,out Result)"}.Result
    ExecuteContext=@n23.True
   ) {variable=ShakeAlpha3}
   > RigVMVariableNode(Value=&RigVMFunction_RandomFloat(Duration=0, Maximum=1.3, Minimum=0.1, Seed=217) {template="Random::Execute(in Seed,in Minimum,in Maximum,in Duration,out Result)"}.Result) {variable=ShakeTimer3}
  ]}
]
hierarchy [
 b_AR_Root {loc=<0 0 0>, rot=<0 0 0.7071 0.7071>}
 Melee_TwoHanded_Handle {parent=b_AR_Root, loc=<8.1477 -0.008 -29.0793>, rot=<0 0 0.7071 -0.7071>}
 dyn_bee_main {parent=Melee_TwoHanded_Handle, loc=<0 0 66>, rot=<0 0 0 -1>}
 dyn_bee_01_body {parent=dyn_bee_main, loc=<-8.0544 -0.7067 -8.978>, rot=<0.6317 -0.6433 -0.3045 0.3074>}
 dyn_bee_01_wing_r {parent=dyn_bee_01_body, loc=<-0.0339 3.0404 0.7233>, rot=<0.2273 0.5141 -0.2207 -0.7971>}
 dyn_bee_02_body {parent=dyn_bee_main, loc=<11.7283 0.8404 -0.4814>, rot=<-0.4058 0.2408 0.6928 -0.5454>}
 dyn_bee_02_wing_l {parent=dyn_bee_02_body, loc=<0.0006 2.9636 -0.7253>, rot=<-0.1519 -0.5149 -0.1808 -0.8241>}
 dyn_bee_03_body {parent=dyn_bee_main, loc=<-13.8785 -5 9.2363>, rot=<0.7159 -0.454 -0.3241 0.4199>}
 dyn_bee_03_wing_r {parent=dyn_bee_03_body, loc=<-0.0213 3.0326 0.6737>, rot=<0.1905 0.5003 -0.2263 -0.8138>}
 dyn_honey_top {parent=dyn_bee_main, loc=<2 0 22>, rot=<0 0 0 1>}
]
variables [
 IsShaking1 {type=bool, default=False}
 IsShaking2 {type=bool, default=False}
 IsShaking3 {type=bool, default=False}
 MaxAngle1 {type=double, default=50}
 MaxAngle2 {type=double, default=50}
 MaxAngle3 {type=double, default=50}
 ShakeAlpha1 {type=double, default=0}
 ShakeAlpha2 {type=double, default=0}
 ShakeAlpha3 {type=double, default=0}
 ShakeTimer1 {type=double, default=0}
 ShakeTimer2 {type=double, default=0}
 ShakeTimer3 {type=double, default=0}
 ShakeValue1 {type=double, default=0}
 ShakeValue2 {type=double, default=0}
 ShakeValue3 {type=double, default=0}
 Time {type=double, default=0}
]
