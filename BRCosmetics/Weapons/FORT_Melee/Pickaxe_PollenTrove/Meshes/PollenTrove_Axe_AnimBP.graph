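# Animation Blueprint for the PollenTrove pickaxe, adding procedural secondary motion to its three bee bodies and honey top.
# The input pose runs through a Control Rig into a cached pose, which feeds two chains of spring-driven Anim Dynamics nodes.
# A bool blend on bIsInFrontEnd picks the stiffer world-space chain in the front end, otherwise a softer, acceleration-clamped chain simulated in component or root-relative space.
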AnimBlueprint /BRCosmetics/Weapons/FORT_Melee/Pickaxe_PollenTrove/Meshes/PollenTrove_Axe_AnimBP
target_skeleton=&/PollenTrove_Axe_Skeleton
graphs [
 AnimGraph {nodes [
   n1 = Anim Dynamics - Bone: dyn_bee_01_body(ComponentPose=Local To Component(LocalPose=Use cached pose 'preDyn'()), Alpha=1, ExternalForce="0,0,0") {
    angularDampingOverride=0.9999
    angularSpringConstant=200
    bAngularSpring=true
    bLinearSpring=true
    bOverrideAngularDamping=true
    bOverrideLinearDamping=true
    bUseGravityOverride=true
    linearDampingOverride=1
    linearSpringConstant=500
    simulationSpace=World
    alphaBoolBlend {blendOption=Linear}
    boundBone {boneName=dyn_bee_01_body}
    componentLinearAccScale=<0.4 0.4 0.4>
    componentLinearVelScale=<0.4 0.4 0.4>
    gravityOverride {z=-980}
    physicsBodyDefinitions [{
      boundBone {boneName=dyn_bee_01_body}
      boxExtents=<500 500 500>
      constraintSetup {angularLimitsMax=<5 5 0>, angularLimitsMin=<-2 -2 -2>, angularTarget=<1 0 0>, linearAxesMax=<0 0.5 0>}
      localJointOffset=<0.4 0 0>
     }]
   }
   n2 = Anim Dynamics - Bone: dyn_bee_02_body(ComponentPose=@n1, Alpha=1, ExternalForce="0,0,0") {
    angularDampingOverride=0.9999
    angularSpringConstant=200
    bAngularSpring=true
    bLinearSpring=true
    bOverrideAngularDamping=true
    bOverrideLinearDamping=true
    bUseGravityOverride=true
    linearDampingOverride=1
    linearSpringConstant=500
    simulationSpace=World
    alphaBoolBlend {blendOption=Linear}
    boundBone {boneName=dyn_bee_02_body}
    componentLinearAccScale=<0.4 0.4 0.4>
    componentLinearVelScale=<0.4 0.4 0.4>
    gravityOverride {z=-980}
    physicsBodyDefinitions [{
      boundBone {boneName=dyn_bee_02_body}
      boxExtents=<500 500 500>
      constraintSetup {angularLimitsMax=<5 5 0>, angularLimitsMin=<-2 -2 -2>, angularTarget=<1 0 0>, linearAxesMax=<0 0.5 0>}
      localJointOffset=<0.4 0 0>
     }]
   }
   n3 = Anim Dynamics - Bone: dyn_bee_03_body(ComponentPose=@n2, Alpha=1, ExternalForce="0,0,0") {
    angularDampingOverride=0.9999
    angularSpringConstant=200
    bAngularSpring=true
    bLinearSpring=true
    bOverrideAngularDamping=true
    bOverrideLinearDamping=true
    bUseGravityOverride=true
    linearDampingOverride=1
    linearSpringConstant=500
    simulationSpace=World
    alphaBoolBlend {blendOption=Linear}
    boundBone {boneName=dyn_bee_03_body}
    componentLinearAccScale=<0.4 0.4 0.4>
    componentLinearVelScale=<0.4 0.4 0.4>
    gravityOverride {z=-980}
    physicsBodyDefinitions [{
      boundBone {boneName=dyn_bee_03_body}
      boxExtents=<500 500 500>
      constraintSetup {angularLimitsMax=<5 5 0>, angularLimitsMin=<-2 -2 -2>, angularTarget=<1 0 0>, linearAxesMax=<0.5 0 0.3>, linearAxesMin=<0 -0.5 0>}
      localJointOffset=<0.4 0 0>
     }]
   }
   n4 = Anim Dynamics - Bone: dyn_honey_top(ComponentPose=@n3, Alpha=1, ExternalForce="0,0,0") {
    angularDampingOverride=1
    angularSpringConstant=300
    bAngularSpring=true
    bLinearSpring=true
    bOverrideAngularDamping=true
    bOverrideLinearDamping=true
    bUseGravityOverride=true
    linearDampingOverride=1
    linearSpringConstant=300
    simulationSpace=World
    alphaBoolBlend {blendOption=Linear}
    boundBone {boneName=dyn_honey_top}
    componentLinearAccScale=<0.4 0.4 0.4>
    componentLinearVelScale=<0.4 0.4 0.4>
    gravityOverride {z=-980}
    physicsBodyDefinitions [{
      boundBone {boneName=dyn_honey_top}
      boxExtents=<500 500 500>
      constraintSetup {angularLimitsMax=<5 5 5>, angularLimitsMin=<-5 -5 -5>, angularTarget=<1 0 0>, linearAxesMax=<0 0.5 0.5>, linearAxesMin=<-0.5 0 0>}
      localJointOffset=<0.4 0 1>
     }]
   }
   n5 = Anim Dynamics - Bone: dyn_bee_01_body(ComponentPose=Local To Component(LocalPose=Use cached pose 'preDyn'()), Alpha=1, ExternalForce="0,0,0") {
    angularDampingOverride=0.9999
    angularSpringConstant=200
    bAngularSpring=true
    bLinearSpring=true
    bOverrideAngularDamping=true
    bOverrideLinearDamping=true
    bUseGravityOverride=true
    linearDampingOverride=1
    linearSpringConstant=50
    alphaBoolBlend {blendOption=Linear}
    boundBone {boneName=dyn_bee_01_body}
    componentAppliedLinearAccClamp=<3000 3000 3000>
    componentLinearAccScale=<0.4 0.4 0.4>
    componentLinearVelScale=<0.4 0.4 0.4>
    gravityOverride {z=-980}
    physicsBodyDefinitions [{
      boundBone {boneName=dyn_bee_01_body}
      boxExtents=<10 10 10>
      constraintSetup {angularLimitsMax=<5 5 0>, angularLimitsMin=<-5 -5 -2>, angularTarget=<1 0 0>}
      localJointOffset=<0.4 0 0>
     }]
   }
   n6 = Anim Dynamics - Bone: dyn_bee_02_body(ComponentPose=@n5, Alpha=1, ExternalForce="0,0,0") {
    angularDampingOverride=0.9999
    angularSpringConstant=200
    bAngularSpring=true
    bLinearSpring=true
    bOverrideAngularDamping=true
    bOverrideLinearDamping=true
    bUseGravityOverride=true
    linearDampingOverride=1
    linearSpringConstant=50
    alphaBoolBlend {blendOption=Linear}
    boundBone {boneName=dyn_bee_02_body}
    componentAppliedLinearAccClamp=<3000 3000 3000>
    componentLinearAccScale=<0.4 0.4 0.4>
    componentLinearVelScale=<0.4 0.4 0.4>
    gravityOverride {z=-980}
    physicsBodyDefinitions [{
      boundBone {boneName=dyn_bee_02_body}
      boxExtents=<10 10 10>
      constraintSetup {angularLimitsMax=<5 5 5>, angularLimitsMin=<-5 -5 -5>, angularTarget=<1 0 0>, linearAxesMax=<0 1 0>, linearAxesMin=<-1.5 0 -2>}
      localJointOffset=<0.4 0 0>
     }]
   }
   n7 = Anim Dynamics - Bone: dyn_bee_03_body(ComponentPose=@n6, Alpha=1, ExternalForce="0,0,0") {
    angularDampingOverride=0.9999
    angularSpringConstant=200
    bAngularSpring=true
    bLinearSpring=true
    bOverrideAngularDamping=true
    bOverrideLinearDamping=true
    bUseGravityOverride=true
    linearDampingOverride=1
    linearSpringConstant=50
    alphaBoolBlend {blendOption=Linear}
    boundBone {boneName=dyn_bee_03_body}
    componentAppliedLinearAccClamp=<3000 3000 3000>
    componentLinearAccScale=<0.4 0.4 0.4>
    componentLinearVelScale=<0.4 0.4 0.4>
    gravityOverride {z=-980}
    physicsBodyDefinitions [{
      boundBone {boneName=dyn_bee_03_body}
      boxExtents=<10 10 10>
      constraintSetup {angularLimitsMax=<5 5 0>, angularLimitsMin=<-5 -5 -2>, angularTarget=<1 0 0>, linearAxesMax=<1.5 0.5 2>}
      localJointOffset=<0.4 0 0>
     }]
   }
   n8 = Anim Dynamics - Bone: dyn_honey_top(ComponentPose=@n7, Alpha=1, ExternalForce="0,0,0") {
    angularDampingOverride=0.9999
    angularSpringConstant=200
    bAngularSpring=true
    bLinearSpring=true
    bOverrideAngularDamping=true
    bOverrideLinearDamping=true
    bUseGravityOverride=true
    linearDampingOverride=1
    linearSpringConstant=50
    simulationSpace=RootRelative
    alphaBoolBlend {blendOption=Linear}
    boundBone {boneName=dyn_honey_top}
    componentAppliedLinearAccClamp=<3000 3000 3000>
    componentLinearAccScale=<0.4 0.4 0.4>
    componentLinearVelScale=<0.4 0.4 0.4>
    gravityOverride {z=-980}
    physicsBodyDefinitions [{
      boundBone {boneName=dyn_honey_top}
      boxExtents=<500 500 500>
      constraintSetup {angularLimitsMax=<5 5 5>, angularLimitsMin=<-5 -5 -5>, angularTarget=<1 0 0>, linearAxesMax=<0 1 2>, linearAxesMin=<-1 0 0>}
      localJointOffset=<0.4 0 1>
     }]
   }
   Output Pose(Result=Blend Poses by bool(BlendPose_0=Component To Local(ComponentPose=@n4), BlendPose_1=Component To Local(ComponentPose=@n8), bActiveValue=Get bIsInFrontEnd()) {blendType=Linear}) {layerGroup=None}
   AnimGraphNode_SubInput_0 = Input Pose()
   n9 = Control Rig(Source=@AnimGraphNode_SubInput_0) {
    alphaBoolBlend {blendOption=Linear}
    controlRigAssetReference {blueprintRigClass=&/PollenTrove_Axe_CtrlRig.PollenTrove_Axe_CtrlRig_C}
   }
   Save cached pose 'preDyn'(Pose=@n9) {cacheName=preDyn}
  ]}
 EventGraph {nodes [
   TryGetPawnOwner()
   Event BlueprintUpdateAnimation()
  ]}
]
variables []
